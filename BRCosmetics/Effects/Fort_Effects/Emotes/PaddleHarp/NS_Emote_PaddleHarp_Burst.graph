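# Niagara system for the PaddleHarp emote: a short, mostly blue burst lasting under a second.
# Mesh emitters supply energy, gem and wave shapes, with sprite glow and slightly delayed rotating shine streaks, all animated by size and color curves over life.
# Wave and glow write extra sprite attributes for layered renderers on one particle.
# Time-of-day scalars adjust several emitters' brightness, front-end checks shrink and dim the glow, and its light-attribute module is disabled.
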NiagaraSystem /BRCosmetics/Effects/Fort_Effects/Emotes/PaddleHarp/NS_Emote_PaddleHarp_Burst
dependencies {
 usedDataInterfaces [&NiagaraDataInterfaceCurve, &NiagaraDataInterfaceColorCurve, &NiagaraDataInterfaceVectorCurve]
 usedDynamicInputs [&MakeVector4_FromLinearColor]
 usedModules [&SystemState]
}
emitters {
 Energy {
  emitterUpdateScript [
   &EmitterState(Life Cycle Mode=::NewEnumerator1, Loop Behavior=::NewEnumerator1, Loop Delay=0.1, Loop Duration=1, UseLoopDelay=false)
   &SpawnBurst_Instantaneous(Spawn Count=1, Use Spawn Probability=false)
  ]
  particleSpawnScript [
   &InitializeParticle(
    Color=<rgba 1 1 1 2>
    Lifetime=0.5
    Mesh Scale Mode=::NewEnumerator3
    Mesh Uniform Scale=0.15
    UsePositionOffset=false
    Write Material Random=true
   ) {module=InitializeParticle001}
   &InitialMeshOrientation(Orientation Vector=<0 0 1>, Rotation=<0 0.25 0>, Use Orientation Vector=true, Use Rotation Vector=true) {module=InitialMeshOrientation001}
   &SystemLocation(MaskSpawnGroup=false)
   &InitialMeshOrientation(
    Rotation=&RandomRangeVector(Minimum=<0 0 0>, Maximum=<1 0 0>)
    Use Orientation Vector=true
    Use Rotation Vector=true
   )
  ]
  particleUpdateScript [
   &ParticleState()
   &ScaleColor(
    Scale Alpha=&Multiply_Float(
     A=&FloatFromCurve(FloatCurve=DataInterface() {propertyValues {Curve {keys [
          {arriveTangent=-0.9, interpMode=RCIM_Cubic, time=0, value=0}
          {arriveTangent=0.8333, interpMode=RCIM_Cubic, leaveTangent=0.8333, tangentMode=RCTM_Break, time=0.1, value=0.5}
          {arriveTangent=-1.6178, interpMode=RCIM_Cubic, leaveTangent=-1.6178, tangentMode=RCTM_User, time=0.6, value=0.5}
          {arriveTangent=-0.5181, interpMode=RCIM_Cubic, leaveTangent=-0.5181, tangentMode=RCTM_User, time=1, value=0}
         ]}}})
     B=&FNTimeOfDayScalar(Night Scale=1, Morning Scale=0.5, Evening Scale=0.5)
    )
    Scale RGB=&MakeVectorFromLinearColorRGB(Color=&ColorFromCurve(ColorCurve=DataInterface() {propertyValues {
        AlphaCurve {keys []}
        BlueCurve {keys [<at 0.1995 0.8259 RCIM_Cubic>, {interpMode=RCIM_Cubic, leaveTangent=0.4741, time=0.4322, value=0.3979}, <at 0.6627 1>]}
        GreenCurve {keys [<at 0.1995 2 RCIM_Cubic>, {interpMode=RCIM_Cubic, leaveTangent=-4.7123, time=0.4322, value=0.7001}, <at 0.6627 0.2695>]}
        RedCurve {keys [<at 0.1995 0.0958 RCIM_Cubic>, {interpMode=RCIM_Cubic, leaveTangent=-0.2609, time=0.4322, value=1}, <at 0.6627 0>]}
       }}))
    ScaleA=true
    ScaleRGB=true
   )
   &ScaleMeshSize(Scale Factor=&ScaleFloatByCurve(
     InitialValue:Vector3f=<1 1 1>
     ScaleCurve:NiagaraDataInterfaceVectorCurve=DataInterface() {propertyValues {XCurve {keys [<at 0 0.25>,<at 1 1>]}, YCurve {keys [<at 0 1.25>,<at 1 2.25>]}, ZCurve {keys [<at 0 1.25>,<at 1 2.25>]}}}
     CurveIndex:NiagaraFloat=$Particles.NormalizedAge
    ))
  ]
  rendererClasses [&NiagaraMeshRendererProperties]
  renderers [{rendererClass=&NiagaraMeshRendererProperties, rendererIndex=0}]
 }
 Wave001 {
  emitterUpdateScript [
   &EmitterState(Life Cycle Mode=::NewEnumerator1, Loop Behavior=::NewEnumerator1, Loop Duration=2, UseLoopDelay=false)
   &SpawnBurst_Instantaneous(Use Spawn Probability=false, UseLoopCountLimit=false)
  ]
  particleSpawnScript [&InitializeParticle(
    Color Mode=::NewEnumerator0
    Lifetime=0.5
    Mesh Scale=<0.1 1 1>
    Mesh Scale Mode=::NewEnumerator4
    UsePositionOffset=false
    Write Material Random=true
   )]
  particleUpdateScript [
   &ParticleState()
   &ScaleMeshSize(Scale Factor=&VectorFromFloat(Value=&FloatFromCurve(
      FloatCurve=DataInterface() {propertyValues {Curve {keys [<at 0 0 RCIM_Cubic>, {arriveTangent=2, interpMode=RCIM_Cubic, leaveTangent=2, time=0.2, value=1.125}, <at 1 2 RCIM_Cubic>]}}}
      Scale Curve=1.4
     )))
   &ScaleColor(
    Scale Alpha=&FloatFromCurve(FloatCurve=DataInterface() {propertyValues {Curve {keys [<at 0 1 RCIM_Cubic>,<at 1 0 RCIM_Cubic>]}}})
    Scale RGB=&Multiply_VectorByFloat(
     Vector=&MakeVectorFromLinearColorRGB(Color=&ColorFromCurve(ColorCurve=DataInterface() {propertyValues {
         AlphaCurve {keys []}
         BlueCurve {keys [<at 0.1393 1 RCIM_Cubic>,<at 0.8768 1 RCIM_Cubic>]}
         GreenCurve {keys [<at 0.1393 0.4645 RCIM_Cubic>,<at 0.8768 0.2067 RCIM_Cubic>]}
         RedCurve {keys [<at 0.1393 0 RCIM_Cubic>,<at 0.8768 0 RCIM_Cubic>]}
        }}))
     Float=&FloatFromCurve(FloatCurve=DataInterface() {propertyValues {Curve {keys [<at 0 1>,<at 1 0>]}}})
    )
    ScaleA=true
    ScaleRGB=true
   )
   &FN_Sprite_Attributes(
    Color=&MakeLinearColorFromVectorAndFloat(
     "Vector (RGB)"=&MakeVectorFromLinearColorRGB(Color=<rgba 0.0378 0.5398 1 1>)
     "Float (Alpha)"=&FloatFromCurve(
      FloatCurve=DataInterface() {propertyValues {Curve {keys [{arriveTangent=-1.9627, interpMode=RCIM_Cubic, leaveTangent=-1.9627, tangentMode=RCTM_User, time=0, value=1}, <at 0.7 0 RCIM_Cubic>]}}}
      Scale Curve=3
     )
    )
    Create Color Attribute=true
    Create SpriteSize Attribute=true
    SpriteSize=&Vector2DFromFloat(Value=&Multiply_Float(
      A=100
      B=&FloatFromCurve(
       FloatCurve=DataInterface() {propertyValues {Curve {keys [
           {arriveTangent=2.4005, interpMode=RCIM_Cubic, leaveTangent=2.4005, tangentMode=RCTM_User, time=0, value=0}
           {arriveTangent=0.9567, interpMode=RCIM_Cubic, leaveTangent=0.9567, tangentMode=RCTM_User, time=0.2, value=0.8}
           <at 1 1.225 RCIM_Cubic>
          ]}}}
       Scale Curve=1.2
      )
     ))
    Write to Module Namespace=true
   )
   &FN_Sprite_Attributes(
    Color=&MakeLinearColorFromVectorAndFloat(
     "Vector (RGB)"=&MakeVectorFromLinearColorRGB(Color=<rgba 0.25 0.6413 1 1>)
     "Float (Alpha)"=&FloatFromCurve(FloatCurve=DataInterface() {propertyValues {Curve {keys [{arriveTangent=-2.9311, interpMode=RCIM_Cubic, leaveTangent=-2.9311, tangentMode=RCTM_User, time=0, value=1}, <at 0.8 0 RCIM_Cubic>]}}})
    )
    Create Color Attribute=true
    Create SpriteSize Attribute=true
    SpriteSize=&Vector2DFromFloat(Value=&Multiply_Float(
      A=110
      B=&FloatFromCurve(
       FloatCurve=DataInterface() {propertyValues {Curve {keys [
           {arriveTangent=1.4794, interpMode=RCIM_Cubic, leaveTangent=1.4794, leaveTangentWeight=0.6621, tangentMode=RCTM_User, time=0, value=0.5}
           {arriveTangent=0.5689, interpMode=RCIM_Cubic, leaveTangent=0.5689, tangentMode=RCTM_User, time=0.2, value=0.85}
           {arriveTangent=0.0369, interpMode=RCIM_Cubic, leaveTangent=0.0369, tangentMode=RCTM_User, time=1, value=1.1}
          ]}}}
       Scale Curve=1.5
      )
     ))
    Write to Module Namespace=true
   ) {module=FN_Sprite_Attributes001}
  ]
  rendererClasses [&NiagaraMeshRendererProperties, &NiagaraSpriteRendererProperties]
  renderers [
   {rendererClass=&NiagaraMeshRendererProperties, rendererIndex=0}
   {rendererClass=&NiagaraSpriteRendererProperties, rendererIndex=1}
   {rendererClass=&NiagaraSpriteRendererProperties, rendererIndex=2}
   {rendererClass=&NiagaraSpriteRendererProperties, rendererIndex=3}
  ]
 }
 gem {
  emitterUpdateScript [
   &EmitterState(Life Cycle Mode=::NewEnumerator1, Loop Behavior=::NewEnumerator1, Loop Duration=2, UseLoopDelay=false)
   &SpawnBurst_Instantaneous(Spawn Count=2, Use Spawn Probability=false, UseLoopCountLimit=false)
   &SpawnRate(Use Spawn Probability=false) {enabled=false}
  ]
  particleSpawnScript [
   &InitializeParticle(Color Mode=::NewEnumerator0, Lifetime=0.5, Mesh Scale Mode=::NewEnumerator3, UsePositionOffset=false, Write Material Random=true)
   &InitialMeshOrientation(Mesh Orientation Mode=::NewEnumerator2, Rotation=<0 0.25 0>, Use Rotation Vector=true)
   &InitialMeshOrientation(
    Mesh Orientation Mode=::NewEnumerator2
    Rotation=&RandomRangeVector(Minimum=<0 0 0>, Maximum=<1 0 0>)
    Use Rotation Vector=true
   ) {module=InitialMeshOrientation001}
  ]
  particleUpdateScript [
   &ParticleState()
   &UpdateMeshOrientation(
    Rotation Rate=&MakeCustomFloatFromBool(Bool=&RandomBool(), True Float=-0.05, False Float=0.05)
    Rotation Vector=<0 0 1>
   )
   &ScaleMeshSize(Scale Factor=&VectorFromFloat(Value=&FloatFromCurve(FloatCurve=DataInterface() {propertyValues {Curve {keys [
          {arriveTangent=3.2954, interpMode=RCIM_Cubic, leaveTangent=3.2954, tangentMode=RCTM_User, time=0, value=0.4}
          {arriveTangent=0.3986, interpMode=RCIM_Cubic, leaveTangent=0.3986, tangentMode=RCTM_User, time=0.2, value=0.9}
          <at 1 1 RCIM_Cubic>
         ]}}})))
   &ScaleColor(
    Scale Alpha=&Multiply_Float(
     A=&FloatFromCurve(FloatCurve=DataInterface() {propertyValues {Curve {keys [
          <at 0 0 RCIM_Cubic>
          {arriveTangent=-1.9784, interpMode=RCIM_Cubic, leaveTangent=-1.9784, tangentMode=RCTM_User, time=0.1, value=1}
          {arriveTangent=-0.4717, interpMode=RCIM_Cubic, leaveTangent=-0.4717, tangentMode=RCTM_User, time=1, value=0}
         ]}}})
     B=&FNTimeOfDayScalar(Night Scale=1, Morning Scale=0.5, Evening Scale=0.5)
    )
    Scale RGB=&Multiply_VectorByFloat(
     Vector=&MakeVectorFromLinearColorRGB(Color=&ColorFromCurve(ColorCurve=DataInterface() {propertyValues {
         AlphaCurve {keys []}
         BlueCurve {keys [<at 0.0978 1>,<at 0.6101 1>]}
         GreenCurve {keys [<at 0.0978 0.3298>,<at 0.6101 0.1902>]}
         RedCurve {keys [<at 0.0978 0.0397>,<at 0.6101 0.092>]}
        }}))
     Float=&FloatFromCurve(FloatCurve=DataInterface() {propertyValues {Curve {keys [
          <at 0 0 RCIM_Cubic>
          {arriveTangent=-0.2569, interpMode=RCIM_Cubic, leaveTangent=-0.2569, tangentMode=RCTM_User, time=0.1, value=10}
          {arriveTangent=-0.1368, interpMode=RCIM_Cubic, leaveTangent=-0.1368, tangentMode=RCTM_User, time=0.5, value=10}
          <at 1 0 RCIM_Cubic>
         ]}}})
    )
    ScaleA=true
    ScaleRGB=true
   ) {module=ScaleColor001}
   &DynamicMaterialParameters(
    Index 0 Param 1=&FloatFromCurve(FloatCurve=DataInterface() {propertyValues {Curve {keys [{arriveTangent=0.6769, time=0, value=0.3}, <at 1 1>]}}}, Scale Curve=0.6)
    Index 0 Param 3=2
    Module.Param0WriteEnabled=true
    Module.Param1WriteEnabled=true
    Module.Param2WriteEnabled=true
    Module.Param3WriteEnabled=false
   ) {module=DynamicMaterialParameters001}
  ]
  rendererClasses [&NiagaraMeshRendererProperties]
  renderers [{rendererClass=&NiagaraMeshRendererProperties, rendererIndex=0}]
 }
 gem02 {
  emitterUpdateScript [
   &EmitterState(Life Cycle Mode=::NewEnumerator1, Loop Behavior=::NewEnumerator1, Loop Delay=0.05, Loop Duration=1, UseLoopDelay=true)
   &SpawnBurst_Instantaneous(Spawn Count=1, Use Spawn Probability=false)
  ]
  particleSpawnScript [
   &InitializeParticle(
    Color=<rgba 1 1 1 4>
    Lifetime=0.4
    Mesh Scale Mode=::NewEnumerator3
    Mesh Uniform Scale=&RandomRangeFloat(Minimum=0.9, Maximum=0.9)
    UsePositionOffset=false
    Write Material Random=true
   )
   &InitialMeshOrientation(Orientation Axis=<0 0 1>, Orientation Vector=<0 0 1>, Rotation=<0 0.25 0>, Use Orientation Vector=true, Use Rotation Vector=true)
   &InitialMeshOrientation(
    Orientation Axis=<0 0 0>
    Orientation Vector=<0 0 0>
    Rotation=&RandomRangeVector(Minimum=<0 0 0>, Maximum=<1 0 0>)
    Use Orientation Vector=false
    Use Rotation Vector=true
   ) {module=InitialMeshOrientation001}
  ]
  particleUpdateScript [
   &ParticleState()
   &ScaleColor(
    Linear Color Curve=DataInterface() {propertyValues {
      AlphaCurve {keys [<at 0 1>,<at 1 0>]}
      BlueCurve {keys [<at 0.0089 1>,<at 0.4813 1>,<at 0.8577 0.1801>]}
      GreenCurve {keys [<at 0.0089 0>,<at 0.4813 0.7483>,<at 0.8577 0.552>]}
      RedCurve {keys [<at 0.0089 0.4528>,<at 0.4813 0.027>,<at 0.8577 1>]}
     }}
    Scale Mode=::NewEnumerator2
   ) {module=ScaleColor002}
   &DynamicMaterialParameters(
    Index 0 Param 1=&FloatFromCurve(FloatCurve=DataInterface() {propertyValues {Curve {keys [{arriveTangent=0.6769, time=0, value=0.3}, <at 1 1>]}}})
    Index 0 Param 2=&FloatFromCurve(FloatCurve=DataInterface() {propertyValues {Curve {keys [{arriveTangent=-0.5818, interpMode=RCIM_Cubic, leaveTangent=-0.5818, tangentMode=RCTM_User, time=0, value=1}, <at 1 0.8>]}}})
    Index 0 Param 3=&FloatFromCurve(FloatCurve=DataInterface() {propertyValues {Curve {keys [<at 0 0>,<at 0.15 1>,<at 1 1.2>]}}})
    Module.Param0WriteEnabled=true
    Module.Param1WriteEnabled=true
    Module.Param2WriteEnabled=true
    Module.Param3WriteEnabled=false
   )
   &ScaleColor(
    Scale Alpha=&Multiply_Float(A=2, B=&FNTimeOfDayScalar(Night Scale=1, Morning Scale=0.5, Evening Scale=0.5))
    Scale RGB=&VectorFromFloat(Value=&FloatFromCurve(FloatCurve=DataInterface() {propertyValues {Curve {keys [<at 0 70>,<at 0.2 3>,<at 1 1>]}}}))
    ScaleA=true
    ScaleRGB=true
   ) {module=ScaleColor001}
   &ScaleMeshSize(Scale Factor=&VectorFromFloat(Value=&FloatFromCurve(FloatCurve=DataInterface() {propertyValues {Curve {keys [<at 0 0.5>,<at 0.15 1>,<at 1 1.2>]}}})))
  ]
  rendererClasses [&NiagaraMeshRendererProperties]
  renderers [{rendererClass=&NiagaraMeshRendererProperties, rendererIndex=0}]
 }
 glow {
  emitterUpdateScript [
   &EmitterState(Inactive Response=::NewEnumerator1, Life Cycle Mode=::NewEnumerator1, Loop Behavior=::NewEnumerator1, Loop Duration=0.5, UseLoopDelay=false)
   &SpawnBurst_Instantaneous(Use Spawn Probability=false, UseLoopCountLimit=false)
  ]
  particleSpawnScript [
   &InitializeParticle(Lifetime=0.6, Sprite Size Mode=::NewEnumerator3, Uniform Sprite Size=100, UsePositionOffset=false, Write Material Random=true)
   &CameraOffset(Camera Offset Amount=5)
  ]
  particleUpdateScript [
   &ParticleState()
   &ScaleSpriteSize(
    Uniform Curve Scale=&MakeCustomFloatFromBool(Bool=$User.bIsFrontEndPreview, True Float=0.5, False Float=1)
    Uniform Curve Sprite Scale=DataInterface() {propertyValues {Curve {keys [<at 0 0.2 RCIM_Cubic>, {arriveTangent=0.05, interpMode=RCIM_Cubic, leaveTangent=0.05, time=0.2, value=0.95}, <at 1 0.25 RCIM_Cubic>]}}}
   )
   /SetVariables(Particles.Size002:Vector2f=&Multiply_Vector2D(A=$Particles.SpriteSize, B=<4 1>)) {isSetParameters=true}
   &ScaleColor(
    Scale Alpha=&FloatFromCurve(
     FloatCurve=DataInterface() {propertyValues {Curve {keys [
         {arriveTangent=8.4201, interpMode=RCIM_Cubic, leaveTangent=8.4201, tangentMode=RCTM_User, time=0, value=0.6}
         {arriveTangent=4.5722, interpMode=RCIM_Cubic, leaveTangent=0.2418, tangentMode=RCTM_Break, time=0.1, value=1.5}
         {arriveTangent=-2.758, interpMode=RCIM_Cubic, leaveTangent=-2.758, tangentMode=RCTM_User, time=0.5, value=0.75}
         <at 1 0 RCIM_Cubic>
        ]}}}
     Scale Curve=&Multiply_Float(
      A=&FNTimeOfDayScalar(Night Scale=0.7, Morning Scale=1, Evening Scale=1)
      B=&MakeCustomFloatFromBool(Bool=$User.bIsFrontEndPreview, True Float=0.5, False Float=1)
     )
    )
    Scale RGB=&MakeVectorFromLinearColorRGB(Color=&ColorFromCurve(ColorCurve=DataInterface() {propertyValues {
        AlphaCurve {keys []}
        BlueCurve {keys [<at 0.1702 0.0938 RCIM_Cubic>]}
        GreenCurve {keys [<at 0.1702 0.0168 RCIM_Cubic>]}
        RedCurve {keys [<at 0.1702 0 RCIM_Cubic>]}
       }}))
    ScaleA=true
    ScaleRGB=true
   )
   /SetVariables(Particles.color002:LinearColor=&MakeLinearColorFromVectorAndFloat(
     "Vector (RGB)"=&MakeVectorFromLinearColorRGB(Color=<rgba 0 1.922 5 1>)
     "Float (Alpha)"=&FloatFromCurve(
      FloatCurve=DataInterface() {propertyValues {Curve {keys [
          {arriveTangent=-2.2858, interpMode=RCIM_Cubic, leaveTangent=-2.2858, tangentMode=RCTM_User, time=0, value=1}
          {arriveTangent=-0.0759, interpMode=RCIM_Cubic, leaveTangent=-0.0759, tangentMode=RCTM_User, time=1, value=0}
         ]}}}
      Scale Curve=&Multiply_Float(
       A=&FNTimeOfDayScalar(Night Scale=0.6, Morning Scale=1, Evening Scale=1)
       B=&FN_FrontEndFloatModifier(Gameplay=1, Front End=0.5)
      )
     )
    )) {isSetParameters=true}
   /SetVariables() {isSetParameters=true}
   &DynamicMaterialParameters(
    Index 0 Param 1=0.75
    Index 0 Param 2=0.5
    Index 0 Param 4=0.5
    Module.Param0WriteEnabled=true
    Module.Param1WriteEnabled=true
    Module.Param2WriteEnabled=true
    Module.Param3WriteEnabled=true
   )
   &FN_Light_Attributes(
    Color=&ColorFromCurve(ColorCurve=DataInterface() {propertyValues {AlphaCurve {keys []}, BlueCurve {keys [<at 0.1896 1.5>]}, GreenCurve {keys [<at 0.1896 0.4976>]}, RedCurve {keys [<at 0.1896 0>]}}})
    Create Color Attribute=true
    Create Exponent Attribute=true
    Create Position Attribute=true
    Create Radius Attribute=true
    Exponent=2
    Radius=&FloatFromCurve(
     FloatCurve=DataInterface() {propertyValues {Curve {keys [<at 0 0.4>,<at 0.2 1.1>,<at 1 1.4>]}}}
     Scale Curve=&MakeCustomFloatFromBool(Bool=$User.bIsFrontEndPreview, True Float=0, False Float=150)
    )
    Scale Color=&FloatFromCurve(
     FloatCurve=DataInterface() {propertyValues {Curve {keys [
         {arriveTangent=4.1988, interpMode=RCIM_Cubic, leaveTangent=4.1988, tangentMode=RCTM_User, time=0, value=0}
         {arriveTangent=-0.0105, interpMode=RCIM_Cubic, leaveTangent=-0.0105, tangentMode=RCTM_User, time=0.15, value=1}
         {arriveTangent=-0.9284, interpMode=RCIM_Cubic, leaveTangent=-0.9284, tangentMode=RCTM_User, time=0.5, value=0}
        ]}}}
     Scale Curve=&Multiply_Float(A=&FNTimeOfDayScalar(Night Scale=0.5, Morning Scale=1, Evening Scale=1), B=20)
    )
   ) {enabled=false}
  ]
  rendererClasses [&NiagaraSpriteRendererProperties, &NiagaraLightRendererProperties]
  renderers [
   {rendererClass=&NiagaraSpriteRendererProperties}
   {rendererClass=&NiagaraSpriteRendererProperties, rendererIndex=1}
   {rendererClass=&NiagaraLightRendererProperties, rendererIndex=2}
  ]
 }
 shine {
  emitterUpdateScript [
   &EmitterState(Life Cycle Mode=::NewEnumerator1, Loop Behavior=::NewEnumerator1, Loop Delay=0.08, Loop Duration=1, UseLoopDelay=true)
   &SpawnBurst_Instantaneous(Spawn Count=10, Use Spawn Probability=false)
  ]
  particleSpawnScript [
   &InitializeParticle(
    Color=&RandomRangeLinearColor(Minimum=<rgba 1 1 1 0.5>, Maximum=<rgba 2 2 2 1.2>)
    Lifetime Max=0.5
    Lifetime Min=0.2
    Lifetime Mode=::NewEnumerator1
    Sprite Rotation Mode=::NewEnumerator2
    Sprite Size Max=<100 200>
    Sprite Size Min=<60 100>
    Sprite Size Mode=::NewEnumerator2
    UsePositionOffset=false
    Write Material Random=true
   )
   &SubUVAnimation(End Frame=3, Start Frame=2, SubUV Animation Mode:ENiagara_SubUVLookupMode=::NewEnumerator2)
  ]
  particleUpdateScript [
   &ParticleState()
   &SpriteRotationRate(Rotation Rate=&Multiply_Float(
     A=100
     B=&FloatFromCurve(FloatCurve=DataInterface() {propertyValues {Curve {keys [<at 0 1>,<at 1 0>]}}})
    ))
   &ScaleSpriteSize(Scale Factor=<0.5 0.6>, Scale Sprite Size Mode=::NewEnumerator2) {module=ScaleSpriteSize001}
   &ScaleSpriteSize(Scale Factor=&MakeVector2D(
     X=&FloatFromCurve(FloatCurve=DataInterface() {propertyValues {Curve {keys [
          {interpMode=RCIM_Cubic, tangentMode=RCTM_User, time=0, value=0}
          {interpMode=RCIM_Cubic, tangentMode=RCTM_User, time=0.2, value=1}
          {tangentMode=RCTM_User, time=1, value=0.5}
         ]}}})
     Y=&FloatFromCurve(FloatCurve=DataInterface() {propertyValues {Curve {keys [
          <at 0 0 RCIM_Cubic>
          {arriveTangent=1.5, interpMode=RCIM_Cubic, leaveTangent=1.5, time=0.2, value=0.9}
          {tangentMode=RCTM_User, time=1, value=1.5}
         ]}}})
    ))
   &ScaleColor(
    Scale Alpha=&FloatFromCurve(FloatCurve=DataInterface() {propertyValues {Curve {keys [<at 0 1>,<at 0.2 1>,<at 1 0>]}}})
    Scale RGB=&Multiply_VectorByFloat(
     Vector=&MakeVectorFromLinearColorRGB(Color=&ColorFromCurve(ColorCurve=DataInterface() {propertyValues {
         AlphaCurve {keys []}
         BlueCurve {keys [<at 0 20>,<at 0.1162 1>,<at 0.7783 1>]}
         GreenCurve {keys [<at 0 10.1602>,<at 0.1162 0.4926>,<at 0.7783 0.13>]}
         RedCurve {keys [<at 0 0.625>,<at 0.1162 0>,<at 0.7783 0.0007>]}
        }}))
     Float=&RandomRangeFloat(Minimum=0.5, Maximum=2)
    )
    ScaleA=true
    ScaleRGB=true
   ) {module=ScaleColor001}
  ]
  rendererClasses [&NiagaraSpriteRendererProperties]
  renderers [{rendererClass=&NiagaraSpriteRendererProperties}]
 }
}
summary {
 systemName=NS_Emote_PaddleHarp_Burst
 emitters [
  {emitterName=glow, rendererClasses [&NiagaraSpriteRendererProperties, &NiagaraLightRendererProperties]}
  {emitterName=Wave001, rendererClasses [&NiagaraMeshRendererProperties, &NiagaraSpriteRendererProperties]}
  {emitterName=shine, rendererClasses [&NiagaraSpriteRendererProperties]}
  {emitterName=gem, rendererClasses [&NiagaraMeshRendererProperties]}
  {emitterName=gem02, rendererClasses [&NiagaraMeshRendererProperties]}
  {emitterName=Energy, rendererClasses [&NiagaraMeshRendererProperties]}
 ]
 userVariables [
  User.bIsFrontEnd {type=NiagaraBool}
  User.bIsFrontEndPreview {type=NiagaraBool}
 ]
}
