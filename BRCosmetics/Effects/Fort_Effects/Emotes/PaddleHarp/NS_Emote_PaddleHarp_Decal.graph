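# Niagara system for the PaddleHarp emote that projects a short-lived decal effect.
# One emitter burst-spawns fixed-size particles that inherit their source's movement and feed two decal renderers.
# Curves over particle life drive two color attributes: red is scaled negative, green and blue stay zero, and alpha fades in, holds and fades out.
# The spawn-rate and random decal orientation modules are disabled.
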NiagaraSystem /BRCosmetics/Effects/Fort_Effects/Emotes/PaddleHarp/NS_Emote_PaddleHarp_Decal
dependencies {usedDataInterfaces [&NiagaraDataInterfaceCurve], usedDynamicInputs [], usedModules [&SystemState]}
emitters {Decal001 {
  emitterUpdateScript [
   &EmitterState(Inactive Response=::NewEnumerator1, Life Cycle Mode=::NewEnumerator1, Loop Behavior=::NewEnumerator1, UseLoopDelay=false)
   &SpawnRate(SpawnRate=5, Use Spawn Probability=false) {enabled=false}
   &SpawnBurst_Instantaneous(Use Spawn Probability=false, UseLoopCountLimit=false)
  ]
  particleSpawnScript [
   &InitializeParticle(
    Lifetime=5
    Sprite Rotation Mode=::NewEnumerator2
    Sprite Size Mode=::NewEnumerator3
    Uniform Sprite Size=200
    UsePositionOffset=false
    Write Material Random=true
   )
   /SetVariables(Particles.DecalSize:Vector3f=<200 260 260>) {isSetParameters=true}
   /Niagara/Modules/Spawn/Decal/InitialDecalOrientation(
    Orientation Vector:Vector3f=<0 0 1>
    Rotation:Vector3f=&RandomRangeVector(Minimum=<0 0 0>, Maximum=<0 0 1>)
    Use Rotation Vector=true
   ) {enabled=false}
  ]
  particleUpdateScript [
   &ParticleState()
   /SetVariables(Particles.DecalColor:LinearColor=/Niagara/DynamicInputs/LinearColor/MakeLinearColorFromVector4(Vector:Vector4f=&MakeVector4(
      X=&FloatFromCurve(FloatCurve=DataInterface() {propertyValues {Curve {keys [<at 0 1>,<at 0.5055 0.8853>,<at 1 0>]}}}, Scale Curve=-1)
      Y=0
      Z=0
      W=&FloatFromCurve(
       FloatCurve=DataInterface() {propertyValues {Curve {keys [
           <at 0 0 RCIM_Cubic>
           {arriveTangent=-0.0431, interpMode=RCIM_Cubic, leaveTangent=-0.0431, tangentMode=RCTM_User, time=0.4, value=1}
           {arriveTangent=0.0066, interpMode=RCIM_Cubic, leaveTangent=0.0066, tangentMode=RCTM_User, time=0.6, value=1}
           <at 1 0 RCIM_Cubic>
          ]}}}
       Scale Curve=5
      )
     ))) {isSetParameters=true}
   /SetVariables(Particles.Linear Color=&MakeLinearColorFromVector4(Vector:Vector4f=&MakeVector4(
      X=&FloatFromCurve(FloatCurve=DataInterface() {propertyValues {Curve {keys [<at 0 1>,<at 0.5055 1>,<at 1 1>]}}}, Scale Curve=-1)
      Y=0
      Z=0
      W=&FloatFromCurve(FloatCurve=DataInterface() {propertyValues {Curve {keys [<at 0 0>,<at 0.1265 1>,<at 0.7005 1>,<at 1 0>]}}}, Scale Curve=3)
     ))) {isSetParameters=true}
   &InheritSourceMovement(ApplyForce=false, ApplyPosition=true, ApplyPositionFalloff=false)
   &SolveForcesAndVelocity(Clamp Velocity=false, Limit Acceleration=false)
  ]
  rendererClasses [&NiagaraDecalRendererProperties]
  renderers [
   {rendererClass=&NiagaraDecalRendererProperties, rendererIndex=0}
   {rendererClass=&NiagaraDecalRendererProperties, rendererIndex=1}
  ]
 }}
summary {
 systemName=NS_Emote_PaddleHarp_Decal
 emitters [{emitterName=Decal001, rendererClasses [&NiagaraDecalRendererProperties]}]
 userVariables []
}
